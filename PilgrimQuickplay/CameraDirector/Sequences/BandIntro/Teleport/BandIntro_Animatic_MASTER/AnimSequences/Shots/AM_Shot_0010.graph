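# Level Sequence for shot 0010 of the BandIntro Teleport animatic, moving a camera rig with an establishing cine camera.
# Named rig actors (transforms, dolly, crane, zoom, offset) each carry a transform track, but only the dolly animates, travelling along X while descending and slightly rotating.
# The cine camera holds a static transform and fixed lens settings.
# A single camera cut section spans the shot, ending before the dolly's last keys.
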LevelSequence /PilgrimQuickplay/CameraDirector/Sequences/BandIntro/Teleport/BandIntro_Animatic_MASTER/AnimSequences/Shots/AM_Shot_0010
bindings [
 CAMERA_crane_Actor {
  id=#1
  possessed_class=Actor
  tracks [{
    type=MovieScene3DTransformTrack
    sections [{
      type=MovieScene3DTransformSection
      channels [
       Location.X {type=MovieSceneScriptingDoubleChannel, default=0, const [0,1]}
       Location.Y {type=MovieSceneScriptingDoubleChannel, default=0, const [0,1]}
       Location.Z {type=MovieSceneScriptingDoubleChannel, default=0, const [0,1]}
       Rotation.X {type=MovieSceneScriptingDoubleChannel, default=0, const [0,1]}
       Rotation.Y {type=MovieSceneScriptingDoubleChannel, default=0, const [0,1]}
       Rotation.Z {type=MovieSceneScriptingDoubleChannel, default=0, const [0,1]}
       Scale.X {type=MovieSceneScriptingDoubleChannel, default=1, const [1,1]}
       Scale.Y {type=MovieSceneScriptingDoubleChannel, default=1, const [1,1]}
       Scale.Z {type=MovieSceneScriptingDoubleChannel, default=1, const [1,1]}
      ]
     }]
   }]
 }
 CAMERA_zoom_Actor {
  id=#2
  possessed_class=Actor
  tracks [{
    type=MovieScene3DTransformTrack
    sections [{
      type=MovieScene3DTransformSection
      channels [
       Location.X {type=MovieSceneScriptingDoubleChannel, default=0, const [0,1]}
       Location.Y {type=MovieSceneScriptingDoubleChannel, default=0, const [0,1]}
       Location.Z {type=MovieSceneScriptingDoubleChannel, default=0, const [0,1]}
       Rotation.X {type=MovieSceneScriptingDoubleChannel, default=0, const [0,1]}
       Rotation.Y {type=MovieSceneScriptingDoubleChannel, default=0, const [0,1]}
       Rotation.Z {type=MovieSceneScriptingDoubleChannel, default=0, const [0,1]}
       Scale.X {type=MovieSceneScriptingDoubleChannel, default=1, const [1,1]}
       Scale.Y {type=MovieSceneScriptingDoubleChannel, default=1, const [1,1]}
       Scale.Z {type=MovieSceneScriptingDoubleChannel, default=1, const [1,1]}
      ]
     }]
   }]
 }
 CineCam_establishing {
  id=#3
  possessed_class=CineCameraActor
  tracks [{
    type=MovieScene3DTransformTrack
    sections [{
      type=MovieScene3DTransformSection
      channels [
       Location.X {type=MovieSceneScriptingDoubleChannel, default=0}
       Location.Y {type=MovieSceneScriptingDoubleChannel, default=0}
       Location.Z {type=MovieSceneScriptingDoubleChannel, default=0}
       Rotation.X {type=MovieSceneScriptingDoubleChannel, default=0}
       Rotation.Y {type=MovieSceneScriptingDoubleChannel, default=0}
       Rotation.Z {type=MovieSceneScriptingDoubleChannel, default=0}
       Scale.X {type=MovieSceneScriptingDoubleChannel, default=1}
       Scale.Y {type=MovieSceneScriptingDoubleChannel, default=1}
       Scale.Z {type=MovieSceneScriptingDoubleChannel, default=1}
      ]
     }]
   }]
 }
 CAMERA_dolly_Actor {
  id=#4
  possessed_class=Actor
  tracks [{
    type=MovieScene3DTransformTrack
    sections [{
      type=MovieScene3DTransformSection
      channels [
       Location.X {type=MovieSceneScriptingDoubleChannel, default=-3402.4181, keys [[39,-7685.0663],[175,-6715.7787]]}
       Location.Y {type=MovieSceneScriptingDoubleChannel, default=0, keys [[39,0],[174,14.0418]]}
       Location.Z {type=MovieSceneScriptingDoubleChannel, default=396.1861, keys [[39,947.9877],[174,599.6719]]}
       Rotation.X {type=MovieSceneScriptingDoubleChannel, default=0, keys [[39,-9.6821],[174,0.8742]]}
       Rotation.Y {type=MovieSceneScriptingDoubleChannel, default=0, keys [[39,-0.2234],[86,0.2692],[174,2.4135]]}
       Rotation.Z {type=MovieSceneScriptingDoubleChannel, default=0, keys [[39,0],[174,0.033]]}
       Scale.X {type=MovieSceneScriptingDoubleChannel, default=1, const [1,1]}
       Scale.Y {type=MovieSceneScriptingDoubleChannel, default=1, const [1,1]}
       Scale.Z {type=MovieSceneScriptingDoubleChannel, default=1, const [1,1]}
      ]
     }]
   }]
 }
 CameraComponent {
  id=#5
  parent_id=#3
  possessed_class=CineCameraComponent
  tracks [
   {
    type=MovieSceneFloatTrack
    display_name=Current Focal Length
    sections [{type=MovieSceneFloatSection, channels [None {type=MovieSceneScriptingFloatChannel, default=35}]}]
   }
   {
    type=MovieSceneFloatTrack
    display_name="Manual Focus Distance (Focus Settings)"
    sections [{type=MovieSceneFloatSection, channels [None {type=MovieSceneScriptingFloatChannel, default=100000}]}]
   }
   {
    type=MovieSceneFloatTrack
    display_name=Current Aperture
    sections [{type=MovieSceneFloatSection, channels [None {type=MovieSceneScriptingFloatChannel, default=2.8}]}]
   }
  ]
 }
 CAMERA_transforms_Actor {
  id=#6
  possessed_class=Actor
  tracks [{
    type=MovieScene3DTransformTrack
    sections [{
      type=MovieScene3DTransformSection
      channels [
       Location.X {type=MovieSceneScriptingDoubleChannel, default=-7147.4231, const [-7147.4233,1]}
       Location.Y {type=MovieSceneScriptingDoubleChannel, default=-624.2845, const [-624.2845,1]}
       Location.Z {type=MovieSceneScriptingDoubleChannel, default=70.7939, const [70.7939,1]}
       Rotation.X {type=MovieSceneScriptingDoubleChannel, default=0, const [0,1]}
       Rotation.Y {type=MovieSceneScriptingDoubleChannel, default=0, const [0,1]}
       Rotation.Z {type=MovieSceneScriptingDoubleChannel, default=176.546, const [176.546,1]}
       Scale.X {type=MovieSceneScriptingDoubleChannel, default=1, const [1,1]}
       Scale.Y {type=MovieSceneScriptingDoubleChannel, default=1, const [1,1]}
       Scale.Z {type=MovieSceneScriptingDoubleChannel, default=1, const [1,1]}
      ]
     }]
   }]
 }
 CAMERA_offset_Actor {
  id=#7
  possessed_class=Actor
  tracks [{
    type=MovieScene3DTransformTrack
    sections [{
      type=MovieScene3DTransformSection
      channels [
       Location.X {type=MovieSceneScriptingDoubleChannel, default=0, const [0,1]}
       Location.Y {type=MovieSceneScriptingDoubleChannel, default=0, const [0,1]}
       Location.Z {type=MovieSceneScriptingDoubleChannel, default=0, const [0,1]}
       Rotation.X {type=MovieSceneScriptingDoubleChannel, default=0, const [0,1]}
       Rotation.Y {type=MovieSceneScriptingDoubleChannel, default=0, const [0,1]}
       Rotation.Z {type=MovieSceneScriptingDoubleChannel, default=0, const [0,1]}
       Scale.X {type=MovieSceneScriptingDoubleChannel, default=1, const [1,1]}
       Scale.Y {type=MovieSceneScriptingDoubleChannel, default=1, const [1,1]}
       Scale.Z {type=MovieSceneScriptingDoubleChannel, default=1, const [1,1]}
      ]
     }]
   }]
 }
]
display_rate {denominator=1, numerator=30}
tick_resolution {denominator=1, numerator=24000}
tracks [{type=MovieSceneCameraCutTrack, sections [{type=MovieSceneCameraCutSection, end_frame=150, start_frame=0}]}]
