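# Level Sequence for shot 0010 of the PilgrimQuickplay band intro animatic (teleport): the portrait establishing camera, with a five-second camera cut section.
# Motion is split across named rig actors (transforms, dolly, crane, zoom, offset): transforms and dolly hold placement, while the others and the CineCamera stay at identity.
# Only the dolly's pitch and yaw animate, easing off early, while focal length widens then pushes into a longer lens; focus and aperture stay fixed.
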LevelSequence /PilgrimQuickplay/CameraDirector/Sequences/BandIntro/Teleport/BandIntro_Animatic_MASTER/AnimSequences/Shots/AM_Shot_0010_Portrait
bindings [
 CineCam_establishing_Portrait {
  id=#1
  possessed_class=CineCameraActor
  tracks [{
    type=MovieScene3DTransformTrack
    sections [{
      type=MovieScene3DTransformSection
      channels [
       Location.X {type=MovieSceneScriptingDoubleChannel, default=0}
       Location.Y {type=MovieSceneScriptingDoubleChannel, default=0}
       Location.Z {type=MovieSceneScriptingDoubleChannel, default=0}
       Rotation.X {type=MovieSceneScriptingDoubleChannel, default=0}
       Rotation.Y {type=MovieSceneScriptingDoubleChannel, default=0}
       Rotation.Z {type=MovieSceneScriptingDoubleChannel, default=0}
       Scale.X {type=MovieSceneScriptingDoubleChannel, default=1}
       Scale.Y {type=MovieSceneScriptingDoubleChannel, default=1}
       Scale.Z {type=MovieSceneScriptingDoubleChannel, default=1}
      ]
     }]
   }]
 }
 CAMERA_crane_Actor {
  id=#2
  possessed_class=Actor
  tracks [{
    type=MovieScene3DTransformTrack
    sections [{
      type=MovieScene3DTransformSection
      channels [
       Location.X {type=MovieSceneScriptingDoubleChannel, default=0, const [0,1]}
       Location.Y {type=MovieSceneScriptingDoubleChannel, default=0, const [0,1]}
       Location.Z {type=MovieSceneScriptingDoubleChannel, default=0, const [0,1]}
       Rotation.X {type=MovieSceneScriptingDoubleChannel, default=0, const [0,1]}
       Rotation.Y {type=MovieSceneScriptingDoubleChannel, default=0, const [0,1]}
       Rotation.Z {type=MovieSceneScriptingDoubleChannel, default=0, const [0,1]}
       Scale.X {type=MovieSceneScriptingDoubleChannel, default=1, const [1,1]}
       Scale.Y {type=MovieSceneScriptingDoubleChannel, default=1, const [1,1]}
       Scale.Z {type=MovieSceneScriptingDoubleChannel, default=1, const [1,1]}
      ]
     }]
   }]
 }
 CAMERA_zoom_Actor {
  id=#3
  possessed_class=Actor
  tracks [{
    type=MovieScene3DTransformTrack
    sections [{
      type=MovieScene3DTransformSection
      channels [
       Location.X {type=MovieSceneScriptingDoubleChannel, default=0, const [0,1]}
       Location.Y {type=MovieSceneScriptingDoubleChannel, default=0, const [0,1]}
       Location.Z {type=MovieSceneScriptingDoubleChannel, default=0, const [0,1]}
       Rotation.X {type=MovieSceneScriptingDoubleChannel, default=0, const [0,1]}
       Rotation.Y {type=MovieSceneScriptingDoubleChannel, default=0, const [0,1]}
       Rotation.Z {type=MovieSceneScriptingDoubleChannel, default=0, const [0,1]}
       Scale.X {type=MovieSceneScriptingDoubleChannel, default=1, const [1,1]}
       Scale.Y {type=MovieSceneScriptingDoubleChannel, default=1, const [1,1]}
       Scale.Z {type=MovieSceneScriptingDoubleChannel, default=1, const [1,1]}
      ]
     }]
   }]
 }
 CameraComponent {
  id=#4
  parent_id=#1
  possessed_class=CineCameraComponent
  tracks [
   {
    type=MovieSceneFloatTrack
    display_name=Current Focal Length
    sections [{
      type=MovieSceneFloatSection
      channels [None {type=MovieSceneScriptingFloatChannel, default=24.3398, keys [[-10,24.3398],[81,15.3383],[112,14.779],[138,18.0368],[198,46.8398]]}]
     }]
   }
   {
    type=MovieSceneFloatTrack
    display_name="Manual Focus Distance (Focus Settings)"
    sections [{type=MovieSceneFloatSection, channels [None {type=MovieSceneScriptingFloatChannel, default=100000}]}]
   }
   {
    type=MovieSceneFloatTrack
    display_name=Current Aperture
    sections [{type=MovieSceneFloatSection, channels [None {type=MovieSceneScriptingFloatChannel, default=2.8, const [2.8,1]}]}]
   }
  ]
 }
 CAMERA_dolly_Actor {
  id=#5
  possessed_class=Actor
  tracks [{
    type=MovieScene3DTransformTrack
    sections [{
      type=MovieScene3DTransformSection
      channels [
       Location.X {type=MovieSceneScriptingDoubleChannel, default=-1834, const [-1423,1]}
       Location.Y {type=MovieSceneScriptingDoubleChannel, default=33, const [-2660,1]}
       Location.Z {type=MovieSceneScriptingDoubleChannel, default=1489, const [947.9877,1]}
       Rotation.X {type=MovieSceneScriptingDoubleChannel, default=0, const [-9.6821,1]}
       Rotation.Y {type=MovieSceneScriptingDoubleChannel, default=26, keys [[-10,41],[90,26]]}
       Rotation.Z {type=MovieSceneScriptingDoubleChannel, default=53, keys [[-10,72],[90,53]]}
       Scale.X {type=MovieSceneScriptingDoubleChannel, default=1, const [1,1]}
       Scale.Y {type=MovieSceneScriptingDoubleChannel, default=1, const [1,1]}
       Scale.Z {type=MovieSceneScriptingDoubleChannel, default=1, const [1,1]}
      ]
     }]
   }]
 }
 CAMERA_transforms_Actor {
  id=#6
  possessed_class=Actor
  tracks [{
    type=MovieScene3DTransformTrack
    sections [{
      type=MovieScene3DTransformSection
      channels [
       Location.X {type=MovieSceneScriptingDoubleChannel, default=-7147.4231, const [-7147.4233,1]}
       Location.Y {type=MovieSceneScriptingDoubleChannel, default=-624.2845, const [-624.2845,1]}
       Location.Z {type=MovieSceneScriptingDoubleChannel, default=70.7939, const [70.7939,1]}
       Rotation.X {type=MovieSceneScriptingDoubleChannel, default=0, const [0,1]}
       Rotation.Y {type=MovieSceneScriptingDoubleChannel, default=0, const [0,1]}
       Rotation.Z {type=MovieSceneScriptingDoubleChannel, default=176.546, const [176.546,1]}
       Scale.X {type=MovieSceneScriptingDoubleChannel, default=1, const [1,1]}
       Scale.Y {type=MovieSceneScriptingDoubleChannel, default=1, const [1,1]}
       Scale.Z {type=MovieSceneScriptingDoubleChannel, default=1, const [1,1]}
      ]
     }]
   }]
 }
 CAMERA_offset_Actor {
  id=#7
  possessed_class=Actor
  tracks [{
    type=MovieScene3DTransformTrack
    sections [{
      type=MovieScene3DTransformSection
      channels [
       Location.X {type=MovieSceneScriptingDoubleChannel, default=0, const [0,1]}
       Location.Y {type=MovieSceneScriptingDoubleChannel, default=0, const [0,1]}
       Location.Z {type=MovieSceneScriptingDoubleChannel, default=0, const [0,1]}
       Rotation.X {type=MovieSceneScriptingDoubleChannel, default=0, const [0,1]}
       Rotation.Y {type=MovieSceneScriptingDoubleChannel, default=0, const [0,1]}
       Rotation.Z {type=MovieSceneScriptingDoubleChannel, default=0, const [0,1]}
       Scale.X {type=MovieSceneScriptingDoubleChannel, default=1, const [1,1]}
       Scale.Y {type=MovieSceneScriptingDoubleChannel, default=1, const [1,1]}
       Scale.Z {type=MovieSceneScriptingDoubleChannel, default=1, const [1,1]}
      ]
     }]
   }]
 }
]
display_rate {denominator=1, numerator=30}
tick_resolution {denominator=1, numerator=24000}
tracks [{type=MovieSceneCameraCutTrack, sections [{type=MovieSceneCameraCutSection, end_frame=150, start_frame=0}]}]
