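# Level Sequence for shot 0020 of the BandIntro Teleport animatic in Pilgrim Quickplay, a single-camera shot at 30 fps.
# A camera cut track spans the shot, and a possessed CineCameraActor follows a keyed location and rotation path, with a second overlapping transform section holding a small rotation offset.
# Its CineCameraComponent holds constant focal length and manual focus distance, while the aperture track has only zero-length, valueless sections.
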LevelSequence /PilgrimQuickplay/CameraDirector/Sequences/BandIntro/Teleport/BandIntro_Animatic_MASTER/AnimSequences/Shots/AM_Shot_0020
bindings [
 CameraComponent {
  id=#1
  parent_id=#2
  possessed_class=CineCameraComponent
  tracks [
   {
    type=MovieSceneFloatTrack
    display_name=Current Focal Length
    sections [
     {type=MovieSceneFloatSection, end_frame=158, start_frame=158, channels [None {type=MovieSceneScriptingFloatChannel}]}
     {type=MovieSceneFloatSection, end_frame=158, start_frame=158, channels [None {type=MovieSceneScriptingFloatChannel}]}
     {type=MovieSceneFloatSection, end_frame=206, start_frame=0, channels [None {type=MovieSceneScriptingFloatChannel, default=105}]}
    ]
   }
   {
    type=MovieSceneFloatTrack
    display_name="Manual Focus Distance (Focus Settings)"
    sections [
     {type=MovieSceneFloatSection, end_frame=158, start_frame=158, channels [None {type=MovieSceneScriptingFloatChannel}]}
     {type=MovieSceneFloatSection, end_frame=158, start_frame=158, channels [None {type=MovieSceneScriptingFloatChannel}]}
     {type=MovieSceneFloatSection, end_frame=210, start_frame=0, channels [None {type=MovieSceneScriptingFloatChannel, default=1755.5647}]}
    ]
   }
   {
    type=MovieSceneFloatTrack
    display_name=Current Aperture
    sections [
     {type=MovieSceneFloatSection, end_frame=158, start_frame=158, channels [None {type=MovieSceneScriptingFloatChannel}]}
     {type=MovieSceneFloatSection, end_frame=158, start_frame=158, channels [None {type=MovieSceneScriptingFloatChannel}]}
     {type=MovieSceneFloatSection, end_frame=158, start_frame=158, channels [None {type=MovieSceneScriptingFloatChannel}]}
    ]
   }
  ]
 }
 Cine Camera Actor_0020 {
  id=#2
  possessed_class=CineCameraActor
  tracks [{
    type=MovieScene3DTransformTrack
    sections [
     {
      type=MovieScene3DTransformSection
      channels [
       Location.X {
        type=MovieSceneScriptingDoubleChannel
        default=-5280
        thinned_from=6
        keys [[-0.0013,-5335.6553],[0,-5271.8736],[111,-5280],[132,-5264.0906],[210,-5250.0945]]
       }
       Location.Y {
        type=MovieSceneScriptingDoubleChannel
        default=-799
        keys [[-0.0013,-754.9424],[35,-751.8716],[53,-793],[132,-799],[153,-756.5457],[210,-754.9424]]
       }
       Location.Z {type=MovieSceneScriptingDoubleChannel, default=578, keys [[0,506.7107],[35,507.5982],[53,567],[132,578],[153,504.6415],[210,506.7107]]}
       Rotation.X {type=MovieSceneScriptingDoubleChannel, default=0, keys [[-0.0013,0],[45,0.0931],[111,-0.934],[170,-0.426],[210,-0.3132]]}
       Rotation.Y {
        type=MovieSceneScriptingDoubleChannel
        default=11.5
        thinned_from=9
        keys [[0,9.5703],[20,13.764],[52,11.5],[136,11.5],[151,9.0613],[160,8.2567],[200,8.0768]]
       }
       Rotation.Z {type=MovieSceneScriptingDoubleChannel, default=174, thinned_from=6, keys [[0,175.8804],[25,175.8265],[52,174],[131,174],[210,175.5988]]}
       Scale.X {type=MovieSceneScriptingDoubleChannel, default=1, const [1,2]}
       Scale.Y {type=MovieSceneScriptingDoubleChannel, default=1, const [1,2]}
       Scale.Z {type=MovieSceneScriptingDoubleChannel, default=1, const [1,2]}
      ]
     }
     {
      type=MovieScene3DTransformSection
      end_frame=179
      start_frame=0
      channels [
       Location.X {type=MovieSceneScriptingDoubleChannel, default=0, const [0,1]}
       Location.Y {type=MovieSceneScriptingDoubleChannel, default=0, const [0,1]}
       Location.Z {type=MovieSceneScriptingDoubleChannel, default=0, const [0,1]}
       Rotation.X {type=MovieSceneScriptingDoubleChannel, default=0, const [0,1]}
       Rotation.Y {type=MovieSceneScriptingDoubleChannel, default=0, const [0.6757,1]}
       Rotation.Z {type=MovieSceneScriptingDoubleChannel, default=0, const [-1.1228,1]}
       Scale.X {type=MovieSceneScriptingDoubleChannel, default=0, const [0,1]}
       Scale.Y {type=MovieSceneScriptingDoubleChannel, default=0, const [0,1]}
       Scale.Z {type=MovieSceneScriptingDoubleChannel, default=0, const [0,1]}
      ]
     }
     {
      type=MovieScene3DTransformSection
      end_frame=158
      start_frame=158
      channels [
       Location.X {type=MovieSceneScriptingDoubleChannel, default=0}
       Location.Y {type=MovieSceneScriptingDoubleChannel, default=0}
       Location.Z {type=MovieSceneScriptingDoubleChannel, default=0}
       Rotation.X {type=MovieSceneScriptingDoubleChannel, default=0}
       Rotation.Y {type=MovieSceneScriptingDoubleChannel, default=0}
       Rotation.Z {type=MovieSceneScriptingDoubleChannel, default=0}
       Scale.X {type=MovieSceneScriptingDoubleChannel, default=1}
       Scale.Y {type=MovieSceneScriptingDoubleChannel, default=1}
       Scale.Z {type=MovieSceneScriptingDoubleChannel, default=1}
      ]
     }
     {
      type=MovieScene3DTransformSection
      end_frame=158
      start_frame=158
      channels [
       Location.X {type=MovieSceneScriptingDoubleChannel, default=0}
       Location.Y {type=MovieSceneScriptingDoubleChannel, default=0}
       Location.Z {type=MovieSceneScriptingDoubleChannel, default=0}
       Rotation.X {type=MovieSceneScriptingDoubleChannel, default=0}
       Rotation.Y {type=MovieSceneScriptingDoubleChannel, default=0}
       Rotation.Z {type=MovieSceneScriptingDoubleChannel, default=0}
       Scale.X {type=MovieSceneScriptingDoubleChannel, default=1}
       Scale.Y {type=MovieSceneScriptingDoubleChannel, default=1}
       Scale.Z {type=MovieSceneScriptingDoubleChannel, default=1}
      ]
     }
    ]
   }]
 }
]
display_rate {denominator=1, numerator=30}
tick_resolution {denominator=1, numerator=24000}
tracks [{type=MovieSceneCameraCutTrack, sections [{type=MovieSceneCameraCutSection, end_frame=210, start_frame=0}]}]
