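# Level Sequence for portrait shot 0020 of the Quickplay band intro "Teleport" animatic, a seven-second single-camera move at 30 fps.
# A camera cut track spans the shot, and a possessed CineCamera follows keyed location and rotation, including abrupt roll changes.
# A second transform section holds small pitch and yaw offsets.
# Focal length and focus distance hold constant values; aperture and other tracks contain zero-length sections at frame 158.
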LevelSequence /PilgrimQuickplay/CameraDirector/Sequences/BandIntro/Teleport/BandIntro_Animatic_MASTER/AnimSequences/Shots/AM_Shot_0020_Portrait
bindings [
 Cine Camera Actor_0020_Portrait {
  id=#1
  possessed_class=CineCameraActor
  tracks [{
    type=MovieScene3DTransformTrack
    sections [
     {
      type=MovieScene3DTransformSection
      channels [
       Location.X {type=MovieSceneScriptingDoubleChannel, default=-4664, thinned_from=4, keys [[-16,-4664],[92,-5280],[210,-5250.0945]]}
       Location.Y {type=MovieSceneScriptingDoubleChannel, default=-787, keys [[-16,-787],[155,-799],[197,-756.5457],[210,-754.9424]]}
       Location.Z {type=MovieSceneScriptingDoubleChannel, default=622, keys [[-16,622],[155,578],[197,504.6415],[210,506.7107]]}
       Rotation.X {type=MovieSceneScriptingDoubleChannel, default=-2, keys [[-16,-2],[-15.9987,25],[92,-0.934],[168,-11],[170,-0.426],[210,-0.3132]]}
       Rotation.Y {type=MovieSceneScriptingDoubleChannel, default=10, thinned_from=6, keys [[-16,10],[45,8],[76,11.5],[159,11.5],[200,8.0768]]}
       Rotation.Z {type=MovieSceneScriptingDoubleChannel, default=175, thinned_from=5, keys [[-16,175],[52,174],[112,174],[210,175.5988]]}
       Scale.X {type=MovieSceneScriptingDoubleChannel, default=1, const [1,1]}
       Scale.Y {type=MovieSceneScriptingDoubleChannel, default=1, const [1,1]}
       Scale.Z {type=MovieSceneScriptingDoubleChannel, default=1, const [1,1]}
      ]
     }
     {
      type=MovieScene3DTransformSection
      end_frame=179
      start_frame=0
      channels [
       Location.X {type=MovieSceneScriptingDoubleChannel, default=0}
       Location.Y {type=MovieSceneScriptingDoubleChannel, default=0}
       Location.Z {type=MovieSceneScriptingDoubleChannel, default=0}
       Rotation.X {type=MovieSceneScriptingDoubleChannel, default=0}
       Rotation.Y {type=MovieSceneScriptingDoubleChannel, default=0.6757}
       Rotation.Z {type=MovieSceneScriptingDoubleChannel, default=-1.1228}
       Scale.X {type=MovieSceneScriptingDoubleChannel, default=0}
       Scale.Y {type=MovieSceneScriptingDoubleChannel, default=0}
       Scale.Z {type=MovieSceneScriptingDoubleChannel, default=0}
      ]
     }
     {
      type=MovieScene3DTransformSection
      end_frame=158
      start_frame=158
      channels [
       Location.X {type=MovieSceneScriptingDoubleChannel, default=0}
       Location.Y {type=MovieSceneScriptingDoubleChannel, default=0}
       Location.Z {type=MovieSceneScriptingDoubleChannel, default=0}
       Rotation.X {type=MovieSceneScriptingDoubleChannel, default=0}
       Rotation.Y {type=MovieSceneScriptingDoubleChannel, default=0}
       Rotation.Z {type=MovieSceneScriptingDoubleChannel, default=0}
       Scale.X {type=MovieSceneScriptingDoubleChannel, default=1}
       Scale.Y {type=MovieSceneScriptingDoubleChannel, default=1}
       Scale.Z {type=MovieSceneScriptingDoubleChannel, default=1}
      ]
     }
     {
      type=MovieScene3DTransformSection
      end_frame=158
      start_frame=158
      channels [
       Location.X {type=MovieSceneScriptingDoubleChannel, default=0}
       Location.Y {type=MovieSceneScriptingDoubleChannel, default=0}
       Location.Z {type=MovieSceneScriptingDoubleChannel, default=0}
       Rotation.X {type=MovieSceneScriptingDoubleChannel, default=0}
       Rotation.Y {type=MovieSceneScriptingDoubleChannel, default=0}
       Rotation.Z {type=MovieSceneScriptingDoubleChannel, default=0}
       Scale.X {type=MovieSceneScriptingDoubleChannel, default=1}
       Scale.Y {type=MovieSceneScriptingDoubleChannel, default=1}
       Scale.Z {type=MovieSceneScriptingDoubleChannel, default=1}
      ]
     }
    ]
   }]
 }
 CameraComponent {
  id=#2
  parent_id=#1
  possessed_class=CineCameraComponent
  tracks [
   {
    type=MovieSceneFloatTrack
    display_name=Current Focal Length
    sections [
     {type=MovieSceneFloatSection, end_frame=158, start_frame=158, channels [None {type=MovieSceneScriptingFloatChannel}]}
     {type=MovieSceneFloatSection, end_frame=158, start_frame=158, channels [None {type=MovieSceneScriptingFloatChannel}]}
     {type=MovieSceneFloatSection, end_frame=206, start_frame=0, channels [None {type=MovieSceneScriptingFloatChannel, default=46.2047}]}
    ]
   }
   {
    type=MovieSceneFloatTrack
    display_name="Manual Focus Distance (Focus Settings)"
    sections [
     {type=MovieSceneFloatSection, end_frame=158, start_frame=158, channels [None {type=MovieSceneScriptingFloatChannel}]}
     {type=MovieSceneFloatSection, end_frame=158, start_frame=158, channels [None {type=MovieSceneScriptingFloatChannel}]}
     {type=MovieSceneFloatSection, end_frame=210, start_frame=0, channels [None {type=MovieSceneScriptingFloatChannel, default=1755.5647}]}
    ]
   }
   {
    type=MovieSceneFloatTrack
    display_name=Current Aperture
    sections [
     {type=MovieSceneFloatSection, end_frame=158, start_frame=158, channels [None {type=MovieSceneScriptingFloatChannel}]}
     {type=MovieSceneFloatSection, end_frame=158, start_frame=158, channels [None {type=MovieSceneScriptingFloatChannel}]}
     {type=MovieSceneFloatSection, end_frame=158, start_frame=158, channels [None {type=MovieSceneScriptingFloatChannel}]}
    ]
   }
  ]
 }
]
display_rate {denominator=1, numerator=30}
tick_resolution {denominator=1, numerator=24000}
tracks [{type=MovieSceneCameraCutTrack, sections [{type=MovieSceneCameraCutSection, end_frame=210, start_frame=0}]}]
