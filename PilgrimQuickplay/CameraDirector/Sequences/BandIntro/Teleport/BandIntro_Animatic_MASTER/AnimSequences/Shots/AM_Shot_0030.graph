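# AM_Shot_0030 is a camera-only Level Sequence, one shot of the BandIntro Teleport animatic for Pilgrim Quickplay's camera director.
# It binds a CineCameraActor whose keyed transform adds subtle handheld-style drift, with a pitch rise early in the shot and a drop later.
# Its camera component holds constant focal length, manual focus distance and aperture.
# A camera cut track spans the seven-second, 30 fps shot.
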LevelSequence /PilgrimQuickplay/CameraDirector/Sequences/BandIntro/Teleport/BandIntro_Animatic_MASTER/AnimSequences/Shots/AM_Shot_0030
bindings [
 CameraComponent {
  id=#1
  parent_id=#2
  possessed_class=CineCameraComponent
  tracks [
   {
    type=MovieSceneFloatTrack
    display_name=Current Focal Length
    sections [{type=MovieSceneFloatSection, channels [None {type=MovieSceneScriptingFloatChannel, default=35}]}]
   }
   {
    type=MovieSceneFloatTrack
    display_name="Manual Focus Distance (Focus Settings)"
    sections [{type=MovieSceneFloatSection, channels [None {type=MovieSceneScriptingFloatChannel, default=100000}]}]
   }
   {
    type=MovieSceneFloatTrack
    display_name=Current Aperture
    sections [{type=MovieSceneFloatSection, channels [None {type=MovieSceneScriptingFloatChannel, default=2.8}]}]
   }
  ]
 }
 CineCam_handHeld_0030 {
  id=#2
  possessed_class=CineCameraActor
  tracks [{
    type=MovieScene3DTransformTrack
    sections [{
      type=MovieScene3DTransformSection
      channels [
       Location.X {
        type=MovieSceneScriptingDoubleChannel
        default=-748.34
        thinned_from=6
        keys [[-0.0013,-5335.6553],[0,-5271.8736],[111,-5280.4067],[132,-5264.0906],[210,-5250.0945]]
       }
       Location.Y {
        type=MovieSceneScriptingDoubleChannel
        default=-1119.3788
        keys [[-0.0013,-754.9424],[35,-751.8716],[89,-755.7467],[153,-756.5457],[210,-754.9424]]
       }
       Location.Z {type=MovieSceneScriptingDoubleChannel, default=523.3652, thinned_from=5, keys [[0,679.3845],[89,682.7694],[153,677.3153],[210,679.3845]]}
       Rotation.X {type=MovieSceneScriptingDoubleChannel, default=0, keys [[-0.0013,0],[45,0.0931],[111,-0.934],[170,-0.426],[210,-0.3132]]}
       Rotation.Y {
        type=MovieSceneScriptingDoubleChannel
        default=-0.2939
        keys [[0,3.202],[20,7.3957],[52,7.4234],[111,7.8354],[139,7.5062],[154,2.6929],[163,1.8884],[203,1.7085]]
       }
       Rotation.Z {type=MovieSceneScriptingDoubleChannel, default=175.0232, thinned_from=5, keys [[0,174.6273],[71,174.3431],[210,174.3457]]}
       Scale.X {type=MovieSceneScriptingDoubleChannel, default=1, const [1,2]}
       Scale.Y {type=MovieSceneScriptingDoubleChannel, default=1, const [1,2]}
       Scale.Z {type=MovieSceneScriptingDoubleChannel, default=1, const [1,2]}
      ]
     }]
   }]
 }
]
display_rate {denominator=1, numerator=30}
tick_resolution {denominator=1, numerator=24000}
tracks [{type=MovieSceneCameraCutTrack, sections [{type=MovieSceneCameraCutSection, end_frame=210, start_frame=0}]}]
