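# Level Sequence for shot 0030 of the BandIntro Teleport animatic in PilgrimQuickplay, named as the Portrait version.
# A possessed CineCamera, labelled handheld, follows keyed location and rotation curves that extend beyond both ends of the camera cut.
# Its focal length pulls from telephoto to very wide early in the shot, while focus distance and aperture stay constant.
# A single camera cut section spans the shot at a 30 fps display rate.
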LevelSequence /PilgrimQuickplay/CameraDirector/Sequences/BandIntro/Teleport/BandIntro_Animatic_MASTER/AnimSequences/Shots/AM_Shot_0030_Portrait
bindings [
 CineCam_handHeld_0030_Portrait {
  id=#1
  possessed_class=CineCameraActor
  tracks [{
    type=MovieScene3DTransformTrack
    sections [{
      type=MovieScene3DTransformSection
      channels [
       Location.X {type=MovieSceneScriptingDoubleChannel, default=-6138, keys [[-35.0013,-5335.6553],[-35,-5271.8736],[16,-5280.4067],[94,-6448],[325,-6138]]}
       Location.Y {type=MovieSceneScriptingDoubleChannel, default=-469, keys [[-35.0013,-754.9424],[-35,-1148],[35,-751.8716],[147,-521],[325,-687]]}
       Location.Z {type=MovieSceneScriptingDoubleChannel, default=811, keys [[-28,874],[35,811],[179,877],[325,1144]]}
       Rotation.X {type=MovieSceneScriptingDoubleChannel, default=0, keys [[-35.0013,0],[45,0.0931],[88,-0.934],[141,-0.3132]]}
       Rotation.Y {type=MovieSceneScriptingDoubleChannel, default=1, keys [[-35,3.202],[88,1],[134,1.7085]]}
       Rotation.Z {type=MovieSceneScriptingDoubleChannel, default=173, keys [[-35,165],[45,173],[160,182],[325,173]]}
       Scale.X {type=MovieSceneScriptingDoubleChannel, default=1, const [1,2]}
       Scale.Y {type=MovieSceneScriptingDoubleChannel, default=1, const [1,2]}
       Scale.Z {type=MovieSceneScriptingDoubleChannel, default=1, const [1,2]}
      ]
     }]
   }]
 }
 CameraComponent {
  id=#2
  parent_id=#1
  possessed_class=CineCameraComponent
  tracks [
   {
    type=MovieSceneFloatTrack
    display_name=Current Focal Length
    sections [{
      type=MovieSceneFloatSection
      channels [None {type=MovieSceneScriptingFloatChannel, default=9.2497, keys [[-35,78.2223],[16,45.4745],[94,9.2497]]}]
     }]
   }
   {
    type=MovieSceneFloatTrack
    display_name="Manual Focus Distance (Focus Settings)"
    sections [{type=MovieSceneFloatSection, channels [None {type=MovieSceneScriptingFloatChannel, default=100000}]}]
   }
   {
    type=MovieSceneFloatTrack
    display_name=Current Aperture
    sections [{type=MovieSceneFloatSection, channels [None {type=MovieSceneScriptingFloatChannel, default=2.8}]}]
   }
  ]
 }
]
display_rate {denominator=1, numerator=30}
tick_resolution {denominator=1, numerator=24000}
tracks [{type=MovieSceneCameraCutTrack, sections [{type=MovieSceneCameraCutSection, end_frame=281, start_frame=0}]}]
